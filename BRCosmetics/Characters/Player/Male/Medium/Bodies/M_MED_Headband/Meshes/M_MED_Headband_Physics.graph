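# Physics asset for the M_MED_Headband male medium player body cosmetic, covering the lower body and a dynamic bag bone.
# Single capsules on the pelvis, both thighs and dyn_bag form the collision bodies.
# Each thigh and the bag is constrained to the pelvis, with the bag held to much tighter swing and twist limits than the legs, which suggests secondary bag motion colliding against the hips and thighs.
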PhysicsAsset /BRCosmetics/Characters/Player/Male/Medium/Bodies/M_MED_Headband/Meshes/M_MED_Headband_Physics
bodies {
 dyn_bag [{
   shapeType=Capsule
   length=9.3999
   radius=5.0637
   shapeName=dyn_bag_capsule
   center=<-6.1998 -6.1399 0.2414>
   rotation=<rot 0 -87.9576 -89.1457>
  }]
 pelvis [{shapeType=Capsule, length=15.8207, radius=12.1284, center=<-2.1823 -0.8843 2.3982>, rotation=<rot -9.4515 -103.0112 12.1767>}]
 thigh_l [{shapeType=Capsule, length=20.5261, radius=10.7064, center=<-20.3213 -0.1289 0.881>, rotation=<rot 0 -82.33 -85.9416>}]
 thigh_r [{shapeType=Capsule, length=20.549, radius=11.4539, center=<20.6161 -0.1313 1.1283>, rotation=<rot 0 -83.949 -79.115>}]
}
constraints [
 {bone1Name=thigh_l, bone2Name=pelvis, swing1LimitDegrees=45, swing2LimitDegrees=45, twistLimitDegrees=45}
 {bone1Name=thigh_r, bone2Name=pelvis, swing1LimitDegrees=45, swing2LimitDegrees=45, twistLimitDegrees=45}
 {bone1Name=dyn_bag, bone2Name=pelvis, swing1LimitDegrees=15, swing2LimitDegrees=10, twistLimitDegrees=2}
]
