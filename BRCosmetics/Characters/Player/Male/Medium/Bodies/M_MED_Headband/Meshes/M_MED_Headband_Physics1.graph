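# Physics asset for the M_MED_Headband medium male player body cosmetic, defining its simplified collision and joint limits.
# Single capsules cover the head, neck, spine, pelvis, arms, hands, legs and feet, plus a dyn_bag bone.
# Only three constraints exist: both thighs and the bag link to the pelvis, with much tighter limits on the bag.
# The limited joint set suggests secondary motion for the bag rather than a full ragdoll.
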PhysicsAsset /BRCosmetics/Characters/Player/Male/Medium/Bodies/M_MED_Headband/Meshes/M_MED_Headband_Physics1
bodies {
 calf_l [{
   shapeType=Capsule
   length=20.25
   radius=9.75
   shapeName=calf_l_capsule
   center=<-17.0704 -0.0004 0.0317>
   rotation=<rot -89.8935 179.9877 -179.9877>
  }]
 calf_r [{
   shapeType=Capsule
   length=20.25
   radius=9.75
   shapeName=calf_r_capsule
   center=<15.2806 0.0001 -0.0645>
   rotation=<rot -89.6412 179.9879 -179.9879>
  }]
 clavicle_l [{shapeType=Capsule, length=20, radius=6.25, shapeName=clavicle_l_capsule, center=<2.1312 0 0.119>, rotation=<rot -86.805 -0.0004 0.0004>}]
 clavicle_r [{
   shapeType=Capsule
   length=20
   radius=6.25
   shapeName=clavicle_r_capsule
   center=<-3.621 -0.0003 -0.2021>
   rotation=<rot -86.806 -0.0004 0.0004>
  }]
 dyn_bag [{
   shapeType=Capsule
   length=7.1499
   radius=7.3137
   shapeName=dyn_bag_capsule
   center=<-6.2792 -3.915 0.2414>
   rotation=<rot 0 -87.9576 -89.1457>
  }]
 foot_l [{shapeType=Capsule, length=9, radius=10.75, shapeName=foot_l_capsule, rotation=<rot 89.205 0.0022 0.0022>}]
 foot_r [{shapeType=Capsule, length=9, radius=10.75, shapeName=foot_r_capsule, rotation=<rot 89.2247 0.002 0.002>}]
 hand_l [{shapeType=Capsule, length=6.25, radius=6.25, shapeName=hand_l_capsule, center=<7.9527 0 0.6541>, rotation=<rot -85.2974 0.0003 -0.0003>}]
 hand_r [{
   shapeType=Capsule
   length=6.5
   radius=6.25
   shapeName=hand_r_capsule
   center=<-4.5813 0.0001 -0.3693>
   rotation=<rot -85.3907 0.0003 -0.0003>
  }]
 head [{shapeType=Capsule, length=0, radius=19.5, shapeName=head_capsule, center=<13.127 0 -0.0003>}]
 lowerarm_l [{
   shapeType=Capsule
   length=28
   radius=6.25
   shapeName=lowerarm_l_capsule
   center=<4.809 0.0001 0.3953>
   rotation=<rot -85.3005 0.0002 -0.0001>
  }]
 lowerarm_r [{
   shapeType=Capsule
   length=20
   radius=6.25
   shapeName=lowerarm_r_capsule
   center=<-7.5482 -0.0002 -0.6204>
   rotation=<rot -85.301 0.0003 -0.0003>
  }]
 neck_01 [{shapeType=Capsule, length=8, radius=7.5, shapeName=neck_01_capsule, rotation=<rot 87.8553 179.9984 179.9984>}]
 pelvis [{shapeType=Capsule, length=15.8207, radius=12.1284, center=<-2.035 -1.0834 1.1416>, rotation=<rot 7.827 -99.3067 -1.6861>}]
 spine_01 [{shapeType=Capsule, length=9.75, radius=15, shapeName=spine_01_capsule}]
 spine_04 [{shapeType=Capsule, length=9.75, radius=15, shapeName=spine_04_capsule}]
 spine_05 [{shapeType=Capsule, length=15, radius=15, shapeName=spine_05_capsule}]
 thigh_l [{shapeType=Capsule, length=20.5261, radius=10.7064, center=<-20.3213 -0.1289 0.881>, rotation=<rot 0 -82.33 -85.9416>}]
 thigh_r [{shapeType=Capsule, length=20.549, radius=11.4539, center=<20.6161 -0.1313 1.1283>, rotation=<rot 0 -83.949 -79.115>}]
 upperarm_l [{
   shapeType=Capsule
   length=20
   radius=6.25
   shapeName=upperarm_l_capsule
   center=<3.7373 -0.0009 -1.2779>
   rotation=<rot -85.298 -0.0001 0.0001>
  }]
 upperarm_r [{
   shapeType=Capsule
   length=20
   radius=6.25
   shapeName=upperarm_r_capsule
   center=<-6.4021 0.0001 1.2992>
   rotation=<rot -85.2985 -0.0001 0.0001>
  }]
}
constraints [
 {bone1Name=thigh_l, bone2Name=pelvis, swing1LimitDegrees=45, swing2LimitDegrees=45, twistLimitDegrees=45}
 {bone1Name=thigh_r, bone2Name=pelvis, swing1LimitDegrees=45, swing2LimitDegrees=45, twistLimitDegrees=45}
 {bone1Name=dyn_bag, bone2Name=pelvis, swing1LimitDegrees=15, swing2LimitDegrees=10, twistLimitDegrees=2}
]
