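# Physics asset for the male medium Headband face accessory, set up for two dangling headband strands, left and right.
# Each strand is a chain of capsules from a base bone through seven segments, constrained link to link and anchored to the head.
# All links share the same limits, with moderate asymmetric swing and tight twist.
# Larger, unconstrained capsules on head, neck, upper spine and clavicles approximate the body as collision shapes.
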PhysicsAsset /BRCosmetics/Characters/Player/Male/Medium/Bodies/M_MED_Headband/Meshes/Parts/M_MED_Headband_FaceAcc_Physics
bodies {
 clavicle_l [{
   shapeType=Capsule
   length=11.046
   radius=8.204
   shapeName=clavicle_l_capsule
   center=<9.6888 0.0701 -0.195>
   rotation=<rot 89.4706 -179.998 -179.998>
  }]
 clavicle_r [{
   shapeType=Capsule
   length=11.046
   radius=8.204
   shapeName=clavicle_r_capsule
   center=<-8.379 0.0704 0.6334>
   rotation=<rot 89.7106 -179.9981 -179.9981>
  }]
 dyn_headband_l_01 [{
   shapeType=Capsule
   length=5.1408
   radius=1.5856
   shapeName=dyn_headband_l_01_capsule
   center=<-4.8331 -0.7945 0.0169>
   rotation=<rot -89.7992 179.985 -179.985>
  }]
 dyn_headband_l_02 [{
   shapeType=Capsule
   length=5.1833
   radius=1.543
   shapeName=dyn_headband_l_02_capsule
   center=<-4.8331 -0.7945 0.0169>
   rotation=<rot -89.7992 179.985 -179.985>
  }]
 dyn_headband_l_03 [{
   shapeType=Capsule
   length=5.5197
   radius=1.7067
   shapeName=dyn_headband_l_03_capsule
   center=<-4.8331 -0.7945 0.0169>
   rotation=<rot -89.7992 179.985 -179.985>
  }]
 dyn_headband_l_04 [{
   shapeType=Capsule
   length=5.6132
   radius=1.6131
   shapeName=dyn_headband_l_04_capsule
   center=<-5.3022 -0.9136 0.0186>
   rotation=<rot -89.7992 179.985 -179.985>
  }]
 dyn_headband_l_05 [{
   shapeType=Capsule
   length=5.1805
   radius=2.0458
   shapeName=dyn_headband_l_05_capsule
   center=<-5.5706 -1.3387 0.0195>
   rotation=<rot -89.7992 179.985 -179.985>
  }]
 dyn_headband_l_06 [{
   shapeType=Capsule
   length=5.1068
   radius=2.1196
   shapeName=dyn_headband_l_06_capsule
   center=<-5.5706 -1.5045 0.0195>
   rotation=<rot -89.7992 179.985 -179.985>
  }]
 dyn_headband_l_07 [{
   shapeType=Capsule
   length=4.9702
   radius=2.2561
   shapeName=dyn_headband_l_07_capsule
   center=<-5.5706 -1.6129 0.0195>
   rotation=<rot -89.7992 179.985 -179.985>
  }]
 dyn_headband_l_base [{
   shapeType=Capsule
   length=3.3752
   radius=1.5856
   shapeName=dyn_headband_l_base_capsule
   center=<-3.3674 -0.7945 0.0118>
   rotation=<rot -89.7992 179.985 -179.985>
  }]
 dyn_headband_r_01 [{
   shapeType=Capsule
   length=5.1408
   radius=1.5856
   shapeName=dyn_headband_r_01_capsule
   center=<-4.8331 -0.7945 0.0169>
   rotation=<rot -89.8002 179.9851 -179.9851>
  }]
 dyn_headband_r_02 [{
   shapeType=Capsule
   length=5.1833
   radius=1.543
   shapeName=dyn_headband_r_02_capsule
   center=<-4.8331 -0.7945 0.0169>
   rotation=<rot -89.8002 179.9851 -179.9851>
  }]
 dyn_headband_r_03 [{
   shapeType=Capsule
   length=5.5197
   radius=1.7067
   shapeName=dyn_headband_r_03_capsule
   center=<-4.8331 -0.7945 0.0169>
   rotation=<rot -89.8002 179.9851 -179.9851>
  }]
 dyn_headband_r_04 [{
   shapeType=Capsule
   length=5.6132
   radius=1.6131
   shapeName=dyn_headband_r_04_capsule
   center=<-5.3022 -0.9136 0.0185>
   rotation=<rot -89.8002 179.9851 -179.9851>
  }]
 dyn_headband_r_05 [{
   shapeType=Capsule
   length=5.1805
   radius=2.0458
   shapeName=dyn_headband_r_05_capsule
   center=<-5.5706 -1.3387 0.0194>
   rotation=<rot -89.8002 179.9851 -179.9851>
  }]
 dyn_headband_r_06 [{
   shapeType=Capsule
   length=5.1068
   radius=2.1196
   shapeName=dyn_headband_r_06_capsule
   center=<-5.5706 -1.9537 0.0194>
   rotation=<rot -89.8002 179.9851 -179.9851>
  }]
 dyn_headband_r_07 [{
   shapeType=Capsule
   length=4.9702
   radius=2.2561
   shapeName=dyn_headband_r_07_capsule
   center=<-5.5706 -2.0622 0.0194>
   rotation=<rot -89.8002 179.9851 -179.9851>
  }]
 dyn_headband_r_base [{
   shapeType=Capsule
   length=3.3752
   radius=1.5856
   shapeName=dyn_headband_r_base_capsule
   center=<-3.3674 -0.7945 0.0117>
   rotation=<rot -89.8002 179.9851 -179.9851>
  }]
 head [{shapeType=Capsule, length=12.7078, radius=9.9916, center=<9.7027 0.2842 0.0933>, rotation=<rot 1.4515 -94.5699 -89.222>}]
 neck_01 [{
   shapeType=Capsule
   length=11.5121
   radius=9.5856
   shapeName=neck_01_capsule
   center=<8.9563 0.6956 0.0722>
   rotation=<rot -89.5729 -0.0017 -12.3859>
  }]
 spine_04 [{shapeType=Capsule, length=15.5439, radius=14.4561, shapeName=spine_04_capsule, center=<3.0749 1.6519 0.0001>}]
 spine_05 [{shapeType=Capsule, length=19.8194, radius=13.6066, shapeName=spine_05_capsule, center=<-0.1529 1.6158 0.0002>, rotation=<rot 0 -0.3865 0>}]
}
constraints [
 {bone1Name=dyn_headband_l_base, bone2Name=head, swing1LimitDegrees=45, swing2LimitDegrees=20, twistLimitDegrees=5}
 {bone1Name=dyn_headband_l_01, bone2Name=dyn_headband_l_base, swing1LimitDegrees=45, swing2LimitDegrees=20, twistLimitDegrees=5}
 {bone1Name=dyn_headband_l_02, bone2Name=dyn_headband_l_01, swing1LimitDegrees=45, swing2LimitDegrees=20, twistLimitDegrees=5}
 {bone1Name=dyn_headband_l_03, bone2Name=dyn_headband_l_02, swing1LimitDegrees=45, swing2LimitDegrees=20, twistLimitDegrees=5}
 {bone1Name=dyn_headband_l_04, bone2Name=dyn_headband_l_03, swing1LimitDegrees=45, swing2LimitDegrees=20, twistLimitDegrees=5}
 {bone1Name=dyn_headband_l_05, bone2Name=dyn_headband_l_04, swing1LimitDegrees=45, swing2LimitDegrees=20, twistLimitDegrees=5}
 {bone1Name=dyn_headband_l_06, bone2Name=dyn_headband_l_05, swing1LimitDegrees=45, swing2LimitDegrees=20, twistLimitDegrees=5}
 {bone1Name=dyn_headband_l_07, bone2Name=dyn_headband_l_06, swing1LimitDegrees=45, swing2LimitDegrees=20, twistLimitDegrees=5}
 {bone1Name=dyn_headband_r_base, bone2Name=head, swing1LimitDegrees=45, swing2LimitDegrees=20, twistLimitDegrees=5}
 {bone1Name=dyn_headband_r_01, bone2Name=dyn_headband_r_base, swing1LimitDegrees=45, swing2LimitDegrees=20, twistLimitDegrees=5}
 {bone1Name=dyn_headband_r_02, bone2Name=dyn_headband_r_01, swing1LimitDegrees=45, swing2LimitDegrees=20, twistLimitDegrees=5}
 {bone1Name=dyn_headband_r_03, bone2Name=dyn_headband_r_02, swing1LimitDegrees=45, swing2LimitDegrees=20, twistLimitDegrees=5}
 {bone1Name=dyn_headband_r_04, bone2Name=dyn_headband_r_03, swing1LimitDegrees=45, swing2LimitDegrees=20, twistLimitDegrees=5}
 {bone1Name=dyn_headband_r_05, bone2Name=dyn_headband_r_04, swing1LimitDegrees=45, swing2LimitDegrees=20, twistLimitDegrees=5}
 {bone1Name=dyn_headband_r_06, bone2Name=dyn_headband_r_05, swing1LimitDegrees=45, swing2LimitDegrees=20, twistLimitDegrees=5}
 {bone1Name=dyn_headband_r_07, bone2Name=dyn_headband_r_06, swing1LimitDegrees=45, swing2LimitDegrees=20, twistLimitDegrees=5}
]
