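# Physics asset for the male medium Headband character's face accessory part, setting up physics bodies for its dynamic hair bones and head.
# One central and four left/right hair bones have capsules, each constrained to its own zero-length base capsule.
# The central joint swings and twists within a small cone, while side joints are near-hinges with slight movement on one swing axis.
# A large head capsule has no constraints.
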PhysicsAsset /Game/Characters/Player/Male/Medium/Bodies/M_MED_Headband/Meshes/Parts/M_MED_Headband_FaceAcc_Physics1
bodies {
 dyn_hair [{shapeType=Capsule, length=0.2941, radius=1.6845, shapeName=dyn_hair_capsule, center=<-0.1165 0 0.0001>, rotation=<rot 0 -90 0>}]
 dyn_hair_base [{shapeType=Capsule, length=0, radius=1.246, shapeName=dyn_hair_base_capsule}]
 dyn_hair_l_a [{
   shapeType=Capsule
   length=3.4336
   radius=0.8549
   shapeName=dyn_hair_l_a_capsule
   center=<-2.0673 0.2696 -0.6557>
   rotation=<rot 0.1571 87.168 -89.4449>
  }]
 dyn_hair_l_a_base [{shapeType=Capsule, length=0, radius=1.246, shapeName=dyn_hair_l_a_base_capsule}]
 dyn_hair_l_b [{
   shapeType=Capsule
   length=5.0342
   radius=0.9118
   shapeName=dyn_hair_l_b_capsule
   center=<-3.079 0.3339 0.1488>
   rotation=<rot 0 -85.718 -81.0243>
  }]
 dyn_hair_l_b_base [{shapeType=Capsule, length=0, radius=1.246, shapeName=dyn_hair_l_b_base_capsule}]
 dyn_hair_r_a [{
   shapeType=Capsule
   length=3.4336
   radius=0.8549
   shapeName=dyn_hair_l_a_capsule
   center=<2.0868 0.064 -0.6154>
   rotation=<rot 0.1571 87.168 -89.4449>
  }]
 dyn_hair_r_a_base [{shapeType=Capsule, length=0, radius=1.246, shapeName=dyn_hair_r_a_base_capsule}]
 dyn_hair_r_b [{
   shapeType=Capsule
   length=5.5088
   radius=0.8543
   shapeName=dyn_hair_r_b_capsule
   center=<2.7261 -0.164 -0.3243>
   rotation=<rot 0 -79.301 -83.1977>
  }]
 dyn_hair_r_b_base [{shapeType=Capsule, length=0, radius=1.246, shapeName=dyn_hair_r_b_base_capsule}]
 head [{shapeType=Capsule, length=6.9538, radius=8.2184, center=<7.6986 2.3768 0.0951>, rotation=<rot 1.5381 -87.6876 -88.7477>}]
}
constraints [
 {bone1Name=dyn_hair, bone2Name=dyn_hair_base, swing1LimitDegrees=12, swing2LimitDegrees=12, twistLimitDegrees=12}
 {bone1Name=dyn_hair_l_a, bone2Name=dyn_hair_l_a_base, swing1LimitDegrees=0.01, swing2LimitDegrees=5, twistLimitDegrees=0.01}
 {bone1Name=dyn_hair_l_b, bone2Name=dyn_hair_l_b_base, swing1LimitDegrees=0.01, swing2LimitDegrees=5, twistLimitDegrees=0.01}
 {bone1Name=dyn_hair_r_a, bone2Name=dyn_hair_r_a_base, swing1LimitDegrees=0.01, swing2LimitDegrees=5, twistLimitDegrees=0.01}
 {bone1Name=dyn_hair_r_b, bone2Name=dyn_hair_r_b_base, swing1LimitDegrees=0.01, swing2LimitDegrees=5, twistLimitDegrees=0.01}
]
