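# Post-process style AnimBP for the male Headband cosmetic's face accessory, adding secondary motion to headband tails and hair over the parent's copied pose.
# Constraints and additive rotations pose the tails from neck angles, movement, front-end and skydiving state.
# Trail controllers, AnimDynamics hair and a final RigidBody follow, with gravity switched for sprinting and skydiving.
# EventGraph updates neck alphas, plus gravity, hair-rotator and backward-movement values the output pose ignores.
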AnimBlueprint /BRCosmetics/Characters/Player/Male/Medium/Bodies/M_MED_Headband/Meshes/Parts/M_MED_Headband_FaceAcc_AnimBP
target_skeleton=&/M_MED_Headband_FaceAcc_Skeleton
graphs [
 AnimGraph {nodes [
   n1 = "Constraint - dyn_headband_l_01 (RR)"!(ComponentPose, Alpha=1, ConstraintWeights_0=0.1, ConstraintWeights_1:float=0.9) -> (Pose) {
    alphaBoolBlend {blendOption=Linear}
    boneToModify {boneName=dyn_headband_l_01}
    constraintSetup [{transformType=Rotation, targetBone {boneName=spine_05}}, {transformType=Rotation, targetBone {boneName=dyn_headband_l_03}}]
    constraintWeights [1,1]
   }
   "Constraint - dyn_headband_r_01 (RR)"!(ComponentPose=@n1, Alpha=1, ConstraintWeights_0=0.1, ConstraintWeights_1:float=0.9) -> (Pose) {
    alphaBoolBlend {blendOption=Linear}
    boneToModify {boneName=dyn_headband_r_01}
    constraintSetup [
     {transformType=Rotation, targetBone {boneName=spine_05}}
     {offsetOption=None, transformType=Rotation, targetBone {boneName=dyn_headband_r_03}}
    ]
    constraintWeights [1,1]
   }
   RigidBody!(ComponentPose, Alpha, OverrideWorldGravity:Vector="0,0,-980", ExternalForce) -> (Pose) {
    bClampLinearTranslationLimitToRefPose=true
    bOverrideWorldGravity=true
    overridePhysicsAsset=/Game/Characters/Player/Male/Medium/Bodies/M_MED_Headband/Meshes/Parts/M_MED_Headband_FaceAcc_Physics1
    simulationSpace=BaseBoneSpace
    alphaBoolBlend {blendOption=Linear}
    baseBoneRef {boneName=neck_01}
    componentAppliedLinearAccClamp=<1500 1500 1500>
    componentLinearAccScale=<0.35 0.3 0.3>
    componentLinearVelScale=<0.4 0.35 0.35>
    overrideWorldGravity {z=-980}
    simSpaceSettings {
     maxAngularAcceleration=1000
     maxAngularVelocity=0.3
     maxLinearAcceleration=700
     maxLinearVelocity=500
     velocityScaleZ=0
     worldAlpha=0.1
     externalLinearDragV=<0.5 0.5>
    }
   }
   n2 = "Constraint - dyn_headband_l_01 (RR)"(
    ComponentPose=Local To Component(LocalPose=Copy Pose From Mesh() {bUseAttachedParent=true})
    Alpha=1
    ConstraintWeights_0=0.8
    ConstraintWeights_1=0.2
   ) {
    alphaBoolBlend {blendOption=Linear}
    boneToModify {boneName=dyn_headband_l_01}
    constraintSetup [{transformType=Rotation, targetBone {boneName=spine_05}}, {transformType=Rotation, targetBone {boneName=dyn_headband_l_03}}]
    constraintWeights [1,1]
   }
   n3 = "Constraint - dyn_headband_r_01 (RR)"(ComponentPose=@n2, Alpha=1, ConstraintWeights_0=0.8, ConstraintWeights_1=0.2) {
    alphaBoolBlend {blendOption=Linear}
    boneToModify {boneName=dyn_headband_r_01}
    constraintSetup [
     {transformType=Rotation, targetBone {boneName=spine_05}}
     {offsetOption=None, transformType=Rotation, targetBone {boneName=dyn_headband_r_03}}
    ]
    constraintWeights [1,1]
   }
   n4 = SelectRotator!(A:Rotator="0,5, 0", B:Rotator="0, 0, 0", bPickA:bool=Get bIsMoving2D()) -> (ReturnValue:Rotator)
   n5 = "Transform (Modify) Bone - Bone: dyn_headband_l_base"(ComponentPose=@n3, Alpha=1, Translation="0,0,0", Rotation=@n4, Scale="1,1,1") {
    rotationMode=BMM_Additive
    rotationSpace=BCS_BoneSpace
    alphaBoolBlend {blendOption=Linear}
    boneToModify {boneName=dyn_headband_l_base}
    rotation {yaw=20}
   }
   n6 = "Transform (Modify) Bone - Bone: dyn_headband_r_base"(ComponentPose=@n5, Alpha=1, Translation="0,0,0", Rotation=@n4, Scale="1,1,1") {
    rotationMode=BMM_Additive
    rotationSpace=BCS_BoneSpace
    alphaBoolBlend {blendOption=Linear}
    boneToModify {boneName=dyn_headband_r_base}
    rotation {yaw=20}
   }
   n7 = Get neck_rot_up() -> (neck_rot_up:double)
   n8 = "Transform (Modify) Bone - Bone: dyn_headband_l_base"(ComponentPose=@n6, Alpha=@n7, Translation="0,0,0", Rotation="0,20,0", Scale="1,1,1") {
    rotationMode=BMM_Additive
    rotationSpace=BCS_BoneSpace
    alphaBoolBlend {blendOption=Linear}
    boneToModify {boneName=dyn_headband_l_base}
    rotation {yaw=20}
   }
   n9 = "Transform (Modify) Bone - Bone: dyn_headband_r_base"(ComponentPose=@n8, Alpha=@n7, Translation="0,0,0", Rotation="0,20,0", Scale="1,1,1") {
    rotationMode=BMM_Additive
    rotationSpace=BCS_BoneSpace
    alphaBoolBlend {blendOption=Linear}
    boneToModify {boneName=dyn_headband_r_base}
    rotation {yaw=20}
   }
   n10 = Get neck_rot_dn() -> (neck_rot_dn:double)
   n11 = "Transform (Modify) Bone - Bone: dyn_headband_l_base"(ComponentPose=@n9, Alpha=@n10, Translation="0,0,0", Rotation="0,15,0", Scale="1,1,1") {
    rotationMode=BMM_Additive
    rotationSpace=BCS_BoneSpace
    alphaBoolBlend {blendOption=Linear}
    boneToModify {boneName=dyn_headband_l_base}
    rotation {yaw=15}
   }
   n12 = "Transform (Modify) Bone - Bone: dyn_headband_r_base"(ComponentPose=@n11, Alpha=@n10, Translation="0,0,0", Rotation="0,15,0", Scale="1,1,1") {
    rotationMode=BMM_Additive
    rotationSpace=BCS_BoneSpace
    alphaBoolBlend {blendOption=Linear}
    boneToModify {boneName=dyn_headband_r_base}
    rotation {yaw=15}
   }
   n13 = SelectRotator(A="0,-10, 0", B="0,-22, 0", bPickA=Get bIsInFrontEnd())
   n14 = "Transform (Modify) Bone - Bone: dyn_headband_l_base"(ComponentPose=@n12, Alpha=1, Translation="0,0,0", Rotation=@n13, Scale="1,1,1") {
    rotationMode=BMM_Additive
    rotationSpace=BCS_BoneSpace
    alphaBoolBlend {blendOption=Linear}
    boneToModify {boneName=dyn_headband_l_base}
    rotation {yaw=-15}
   }
   n15 = "Transform (Modify) Bone - Bone: dyn_headband_r_base"(ComponentPose=@n14, Alpha=1, Translation="0,0,0", Rotation=@n13, Scale="1,1,1") {
    rotationMode=BMM_Additive
    rotationSpace=BCS_BoneSpace
    alphaBoolBlend {blendOption=Linear}
    boneToModify {boneName=dyn_headband_r_base}
    rotation {yaw=-15}
   }
   n16 = Get SkydivingAlpha() -> (SkydivingAlpha:Rotator)
   n17 = "Transform (Modify) Bone - Bone: dyn_headband_l_base"(ComponentPose=@n15, Alpha=1, Translation="0,0,0", Rotation=@n16, Scale="1,1,1") {
    rotationMode=BMM_Additive
    rotationSpace=BCS_BoneSpace
    alphaBoolBlend {blendOption=Linear}
    boneToModify {boneName=dyn_headband_l_base}
    rotation {yaw=-15}
   }
   n18 = "Transform (Modify) Bone - Bone: dyn_headband_r_base"(ComponentPose=@n17, Alpha=1, Translation="0,0,0", Rotation=@n16, Scale="1,1,1") {
    rotationMode=BMM_Additive
    rotationSpace=BCS_BoneSpace
    alphaBoolBlend {blendOption=Linear}
    boneToModify {boneName=dyn_headband_r_base}
    rotation {yaw=-15}
   }
   n19 = SelectFloat(A=0, B=SelectFloat(A=0.7, B=0.5, bPickA=Get bIsMoving2D()), bPickA=Get bIsBackpedaling() -> (bIsBackpedaling:bool))
   n20 = Trail controller - Bone: dyn_headband_l_07(ComponentPose=@n18, Alpha=@n19) {
    bInvertChainBoneAxis=true
    bLimitRotation=true
    bUsePlanarLimit=true
    chainLength=7
    debugLifeTime=0.1
    maxDeltaTime=0.333
    alphaScaleBiasClamp {bInterpResult=true, scale=0.65}
    baseJoint {boneName=root}
    rotationLimits [
     {limitMax=<180 180 180>}
     {limitMax=<180 180 180>}
     {limitMax=<180 180 180>}
     {limitMax=<180 180 180>}
     {limitMax=<180 180 180>}
     {limitMax=<180 180 180>}
     {limitMax=<180 180 180>}
    ]
    rotationOffsets [<0 0 0>, <0 0 0>, <0 0 0>, <0 0 0>, <0 0 0>, <0 0 0>, <0 0 0>]
    trailBone {boneName=dyn_headband_l_07}
    trailRelaxationSpeed {editorCurveData {keys [<at -0.0134 19.3841 RCIM_Linear>,<at 0.4187 12.9226 RCIM_Linear>,<at 0.9955 9.6044 RCIM_Linear>]}}
   }
   n21 = Trail controller - Bone: dyn_headband_r_07(ComponentPose=@n20, Alpha=@n19) {
    bInvertChainBoneAxis=true
    bLimitRotation=true
    bUsePlanarLimit=true
    chainLength=7
    debugLifeTime=0.1
    maxDeltaTime=0.333
    alphaScaleBiasClamp {bInterpResult=true, scale=0.65}
    baseJoint {boneName=root}
    rotationLimits [
     {limitMax=<180 180 180>}
     {limitMax=<180 180 180>}
     {limitMax=<180 180 180>}
     {limitMax=<180 180 180>}
     {limitMax=<180 180 180>}
     {limitMax=<180 180 180>}
     {limitMax=<180 180 180>}
    ]
    rotationOffsets [<0 0 0>, <0 0 0>, <0 0 0>, <0 0 0>, <0 0 0>, <0 0 0>, <0 0 0>]
    trailBone {boneName=dyn_headband_r_07}
    trailRelaxationSpeed {editorCurveData {keys [<at -0.0134 19.3841 RCIM_Linear>,<at 0.4899 10.7965 RCIM_Linear>,<at 0.9955 9.6044 RCIM_Linear>]}}
   }
   n22 = "Transform (Modify) Bone - Bone: dyn_hair_l_a"(ComponentPose=@n21, Alpha=1, Translation="0,0,0", Rotation="5,0,0", Scale="1,1,1") {
    rotationMode=BMM_Additive
    rotationSpace=BCS_BoneSpace
    translationMode=BMM_Additive
    translationSpace=BCS_BoneSpace
    alphaBoolBlend {blendOption=Linear}
    boneToModify {boneName=dyn_hair_l_a}
    rotation {pitch=5}
   }
   n23 = Anim Dynamics - Bone: dyn_hair_l_a(ComponentPose=@n22, Alpha=0.7, ExternalForce="0,0,0") {
    angularDampingOverride=0.999
    bAngularSpring=true
    bGravityOverrideInSimSpace=true
    bOverrideAngularDamping=true
    bUseGravityOverride=true
    simulationSpace=RootRelative
    alphaBoolBlend {blendOption=Linear}
    boundBone {boneName=dyn_hair_l_a}
    componentAppliedLinearAccClamp=<800 800 800>
    componentLinearAccScale=<0.3 0.3 0.35>
    componentLinearVelScale=<0.35 0.35 0.35>
    gravityOverride {z=-800}
    physicsBodyDefinitions [{
      sphereCollisionRadius=0
      boundBone {boneName=dyn_hair_l_a}
      boxExtents=<5 15 15>
      constraintSetup {angularTargetAxis=AxisY, angularLimitsMax=<30 20 20>, angularLimitsMin=<-30 -20 -20>, angularTarget=<0 0 0>}
      localJointOffset=<-5 0 0>
     }]
    relativeSpaceBone {boneName=neck_01}
   }
   n24 = "Transform (Modify) Bone - Bone: dyn_hair_l_b"(ComponentPose=@n23, Alpha=1, Translation="0,0,0", Rotation="10,0,0", Scale="1,1,1") {
    rotationMode=BMM_Additive
    rotationSpace=BCS_BoneSpace
    translationMode=BMM_Additive
    translationSpace=BCS_BoneSpace
    alphaBoolBlend {blendOption=Linear}
    boneToModify {boneName=dyn_hair_l_b}
    rotation {pitch=10}
   }
   n25 = Anim Dynamics - Bone: dyn_hair_l_b(ComponentPose=@n24, Alpha=0.7, ExternalForce="0,0,0") {
    angularDampingOverride=0.999
    bAngularSpring=true
    bGravityOverrideInSimSpace=true
    bOverrideAngularDamping=true
    bUseGravityOverride=true
    simulationSpace=RootRelative
    alphaBoolBlend {blendOption=Linear}
    boundBone {boneName=dyn_hair_l_b}
    componentAppliedLinearAccClamp=<800 800 800>
    componentLinearAccScale=<0.3 0.3 0.35>
    componentLinearVelScale=<0.35 0.35 0.35>
    gravityOverride {z=-800}
    physicsBodyDefinitions [{
      sphereCollisionRadius=0
      boundBone {boneName=dyn_hair_l_b}
      boxExtents=<5 15 15>
      constraintSetup {angularTargetAxis=AxisY, angularLimitsMax=<30 20 20>, angularLimitsMin=<-30 -20 -20>, angularTarget=<0 0 0>}
      localJointOffset=<-5 0 0>
     }]
    relativeSpaceBone {boneName=neck_01}
   }
   n26 = "Transform (Modify) Bone - Bone: dyn_hair_r_a"(ComponentPose=@n25, Alpha=1, Translation="0,0,0", Rotation="5,0,0", Scale="1,1,1") {
    rotationMode=BMM_Additive
    rotationSpace=BCS_BoneSpace
    translationMode=BMM_Additive
    translationSpace=BCS_BoneSpace
    alphaBoolBlend {blendOption=Linear}
    boneToModify {boneName=dyn_hair_r_a}
    rotation {pitch=5}
   }
   n27 = Anim Dynamics - Bone: dyn_hair_r_a(ComponentPose=@n26, Alpha=0.7, ExternalForce="0,0,0") {
    angularDampingOverride=0.999
    bAngularSpring=true
    bGravityOverrideInSimSpace=true
    bOverrideAngularDamping=true
    bUseGravityOverride=true
    simulationSpace=RootRelative
    alphaBoolBlend {blendOption=Linear}
    boundBone {boneName=dyn_hair_r_a}
    componentAppliedLinearAccClamp=<800 800 800>
    componentLinearAccScale=<0.3 0.3 0.35>
    componentLinearVelScale=<0.35 0.35 0.35>
    gravityOverride {z=-800}
    physicsBodyDefinitions [{
      sphereCollisionRadius=0
      boundBone {boneName=dyn_hair_r_a}
      boxExtents=<5 15 15>
      constraintSetup {angularTargetAxis=AxisY, angularLimitsMax=<30 20 20>, angularLimitsMin=<-30 -20 -20>, angularTarget=<0 0 0>}
      localJointOffset=<5 0 0>
     }]
    relativeSpaceBone {boneName=neck_01}
   }
   n28 = "Transform (Modify) Bone - Bone: dyn_hair_r_b"(ComponentPose=@n27, Alpha=1, Translation="0,0,0", Rotation="10,0,0", Scale="1,1,1") {
    rotationMode=BMM_Additive
    rotationSpace=BCS_BoneSpace
    translationMode=BMM_Additive
    translationSpace=BCS_BoneSpace
    alphaBoolBlend {blendOption=Linear}
    boneToModify {boneName=dyn_hair_r_b}
    rotation {pitch=10}
   }
   n29 = Anim Dynamics - Bone: dyn_hair_r_b(ComponentPose=@n28, Alpha=0.7, ExternalForce="0,0,0") {
    angularDampingOverride=0.999
    bAngularSpring=true
    bGravityOverrideInSimSpace=true
    bOverrideAngularDamping=true
    bUseGravityOverride=true
    simulationSpace=RootRelative
    alphaBoolBlend {blendOption=Linear}
    boundBone {boneName=dyn_hair_r_b}
    componentAppliedLinearAccClamp=<800 800 800>
    componentLinearAccScale=<0.3 0.3 0.35>
    componentLinearVelScale=<0.35 0.35 0.35>
    gravityOverride {z=-800}
    physicsBodyDefinitions [{
      sphereCollisionRadius=0
      boundBone {boneName=dyn_hair_r_b}
      boxExtents=<5 15 15>
      constraintSetup {angularTargetAxis=AxisY, angularLimitsMax=<30 20 20>, angularLimitsMin=<-30 -20 -20>, angularTarget=<0 0 0>}
      localJointOffset=<5 0 0>
     }]
    relativeSpaceBone {boneName=neck_01}
   }
   n30 = Anim Dynamics - Bone: dyn_hair(ComponentPose=@n29, Alpha=0.7, ExternalForce="0,0,0") {
    angularDampingOverride=0.999
    bAngularSpring=true
    bGravityOverrideInSimSpace=true
    bOverrideAngularDamping=true
    bUseGravityOverride=true
    simulationSpace=RootRelative
    alphaBoolBlend {blendOption=Linear}
    boundBone {boneName=dyn_hair}
    componentAppliedLinearAccClamp=<800 800 800>
    componentLinearAccScale=<0.3 0.3 0.35>
    componentLinearVelScale=<0.35 0.35 0.35>
    gravityOverride {z=-800}
    physicsBodyDefinitions [{
      sphereCollisionRadius=0
      boundBone {boneName=dyn_hair}
      boxExtents=<15 15 15>
      constraintSetup {angularTargetAxis=AxisY, angularLimitsMax=<5 20 20>, angularLimitsMin=<-5 -20 -20>, angularTarget=<0 0 0>}
      localJointOffset=<-5 0 0>
     }]
    relativeSpaceBone {boneName=neck_01}
   }
   n31 = Get bIsSprinting()
   n32 = Get bIsSkyDiving()
   n33 = Get bIsSkydiveFloating()
   n34 = RigidBody(
    ComponentPose=@n30
    OverrideWorldGravity=SelectVector(A="0,0,200", B=SelectVector(A="0,0,-500", B="0, 0,-750", bPickA=@n31), bPickA=OR Boolean!(A=@n33, B=@n32, C:bool) -> (ReturnValue))
   ) {
    bClampLinearTranslationLimitToRefPose=true
    bOverrideWorldGravity=true
    overridePhysicsAsset=&/M_MED_Headband_FaceAcc_Physics
    simulationSpace=BaseBoneSpace
    alphaBoolBlend {blendOption=Linear}
    baseBoneRef {boneName=neck_01}
    componentAppliedLinearAccClamp=<1200 1200 1200>
    componentLinearAccScale=<0.3 0.3 0.3>
    componentLinearVelScale=<0.35 0.35 0.35>
    overrideWorldGravity {z=-980}
    simSpaceSettings {
     maxAngularAcceleration=1000
     maxAngularVelocity=3
     maxLinearAcceleration=700
     maxLinearVelocity=500
     velocityScaleZ=0
     worldAlpha=1
     externalLinearDragV=<0.2 0.2>
    }
   }
   Output Pose(Result=Component To Local(ComponentPose=@n34)) {layerGroup=None}
   Get BwdMovement() -> (BwdMovement:double)
   Get GravityOverride()
  ]}
 EventGraph {nodes [
   Event BlueprintUpdateAnimation()
   > n1 = Sequence!(execute) -> (then_0, then_1, then_2:Exec, then_3:Exec, then_4:Exec)
   > n2 = Gravity Override State Function DEPRECATED(CharacterPartAnimInstance=Self-Reference(), GameTimeInSeconds=GetGameTimeInSeconds(), GravityOverrideParams=Get Gravity Override Params())
   > Set GravityOverride(GravityOverride=@n2)
   > Set GravityOverride_FrontHair(GravityOverride_FrontHair:Vector=vector * vector(A:Vector="0, 0, 0", B:Vector="0.4,0.4,0.7") -> (ReturnValue:Vector) {is=title K2Node_CommutativeAssociativeBinaryOperator}) -> (Output_Get:Vector)
   n3 = GetGameTimeInSeconds()
   n4 = float * float(A=float * float(A=0.4, B=0.7), B=@n3)
   n5 = RandomRotatorFunction!(
    execute=@n1.then_1
    GameTimeInSeconds:double=float * float(A="Sin (Radians)"(A=@n4), B=0)
    PerlinScale:double=1
    Frequency:Rotator="0,-5, 0"
    Amplitude:Rotator="1,2,0"
    Offset:Rotator="0,2, 0"
    TimeOffset:double=0
   ) -> (then, RandomRotator:Rotator)
   > Set HairRotator_01(HairRotator_01:Rotator=@n5) -> (Output_Get:Rotator)
   n6 = float * float(A=float * float(A=0.4, B=0.6), B=@n3)
   > n7 = RandomRotatorFunction(
    GameTimeInSeconds=float * float(A="Sin (Radians)"(A=@n6), B=0)
    PerlinScale=1
    Frequency="1.1,-2,0"
    Amplitude="1,5,0"
    Offset="0, 0, 0"
    TimeOffset=0
   )
   > Set HairRotator_02(HairRotator_02:Rotator=@n7) -> (Output_Get:Rotator)
   n8 = Branch(execute=@n1.then_4, Condition=Get bIsSkyDiving())
   > n9 = Branch(Condition="float > float"(A=Get PawnVelocityZ(), B=50) {is=title K2Node_CallFunction})
   > Set BwdMovement(BwdMovement:double=float * float(A=Get BwdMovement(), B=0)) -> (Output_Get:double)
   Set BwdMovement(execute=[@n8.else, @n9.else], BwdMovement=float * float(A=Get BwdMovement(), B=1))
   n10 = Get Skel_Mesh()
   Set neck_rot_up(
    execute=@n1.then_2
    neck_rot_up:double=GetAlphaFromReferenceTransformDelta(SkeletalMeshComponent=@n10, BoneName=head, BaseName=spine_05, InputRange=(X=-50,Y=-90), TwistAxis="0,0,1")
   ) -> (Output_Get:double)
   > Set neck_rot_dn(neck_rot_dn:double=GetAlphaFromReferenceTransformDelta(SkeletalMeshComponent=@n10, BoneName=neck_01, BaseName=spine_05, DeltaComponentType=EulerX, InputRange=(X=-15,Y=50), TwistAxis="0,0,1")) -> (Output_Get:double)
   n11 = PlayerMovement(execute=@n1.then_3, AnimInstance=Self-Reference(), XDotProductCoefficient=1, YDotProductCoefficient=1, ZDotProductCoefficient=1)
   n12 = MapRangeClamped(Value=@n11.YMovement, InRangeA=-0.2, InRangeB=0.4, OutRangeA=-0.2, OutRangeB=0.2)
   n13 = "Absolute (Float)"(A=@n11)
   > Set BwdMovement(BwdMovement=MapRangeClamped(Value=float + float(A=@n13, B=@n12) {is=title K2Node_CommutativeAssociativeBinaryOperator}, InRangeA=0, InRangeB=1.4))
   Event OnBeginSkydiving()
   > Set SkydivingAlpha(SkydivingAlpha:Rotator="0,30,0") -> (Output_Get:Rotator)
   Event BlueprintInitializeAnimation()
   > Set Skel_Mesh(Skel_Mesh=GetOwnerSkeletalMeshComponent())
   Event OnEndSkydiving()
   > Set SkydivingAlpha(SkydivingAlpha="0,10,0")
   MakeRotator(Yaw=float * float(A=Get neck_rot_dn(), B=2))
   Get Gravity Override Params_01() -> (Gravity Override Params_01:Gravity Override Params Struct Structure)
  ]}
]
variables [
 BwdMovement {type=number}
 Gravity Override Params {
  type=GravityOverrideParamsStruct
  default {front_End {gravityOverride=<0 0 -500>}, onGround_Standing {gravityOverride=<0 0 -500>}, skydive_Parachuting {gravityOverride=<0 -800 -500>}}
 }
 Gravity Override Params_01 {
  type=GravityOverrideParamsStruct
  default {
   emoteOrMelee {gravityOverride=<0 -50 -25>, windAmplitude=<50 25 50>, windFrequency=<0.1 0.2 0.15>, windOffset=<0 -50 -50>}
   front_End {gravityOverride=<0 -50 -15>, windAmplitude=<50 25 50>, windFrequency=<0.1 0.2 0.15>, windOffset=<0 -50 -50>}
   onGround_Moving {gravityOverride=<0 -50 -15>, windAmplitude=<50 25 50>, windFrequency=<0.1 0.2 0.15>, windOffset=<0 -50 -50>}
   onGround_Standing {gravityOverride=<0 -50 -55>, windAmplitude=<50 25 50>, windFrequency=<0.1 0.2 0.15>, windOffset=<0 -50 -50>}
   skydive_Parachuting {gravityOverride=<0 -300 -300>}
   swimming {windAmplitude=<50 -25 100>, windFrequency=<0.15 1 0.3>, windOffset=<0 -50 -50>}
  }
 }
 GravityOverride {type=Vector}
 GravityOverride_FrontHair {type=Vector}
 HairRotator_01 {type=Rotator}
 HairRotator_02 {type=Rotator}
 neck_rot_dn {type=number}
 neck_rot_up {type=number}
 Skel_Mesh {type=SkeletalMeshComponent}
 SkydivingAlpha {type=Rotator}
]
